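# Library material function for UI materials that remaps a 0-1 scalar with circular ease curves, output as In, Out and In-Out.
# The input defaults to the horizontal normalized UI UV and feeds square-root circle-arc formulas, each saturated.
# In-Out avoids branching by mirroring the input around its midpoint, signing the arc with a rounded lerp, then rescaling to 0-1.
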MaterialFunction /FortUILibrary/MaterialFunctions/Basics/MF_UI_EaseCurves_06_Circular
graph {
 properties {bExposeToLibrary=true, description=Outputs interpolated values using ease curves., previewMaterialDomain=MD_UI}
 nodes [
  Constant() {r=-0.5}
  Frac(None=Time())
  n1 = NamedRerouteUsage()
  n2 = SquareRoot(None=OneMinus(None=Multiply(A=@n1, B=@n1)))
  FunctionOutput(None=Saturate(None=OneMinus(None=@n2))) {description=Ouputs an EaseIn, outputName=In, sortPriority=20}
  n3 = OneMinus(None=NamedRerouteUsage())
  n4 = OneMinus(None=Multiply(A=@n3, B=@n3))
  FunctionOutput(None=Saturate(None=SquareRoot(None=@n4))) {description=Ouputs an EaseOut, outputName=Out, sortPriority=21}
  n5 = NamedRerouteUsage()
  n6 = OneMinus(None=Abs(None=ConstantBiasScale(None=@n5) {bias=-0.5, scale=2}))
  n7 = Multiply(A=@n6, B=@n6)
  n8 = Multiply(A=LinearInterpolate(Alpha=Round(None=@n5)) {constA=-1}, B=SquareRoot(None=OneMinus(None=@n7)))
  FunctionOutput(None=Saturate(None=ConstantBiasScale(None=@n8))) {description=Ouputs an EaseInOut, outputName=In-Out, sortPriority=22}
  n9 = FunctionInput(Preview=ComponentMask(None=MaterialFunctionCall() {materialFunction=&GetUserInterfaceUV}.Normalized UV) {r=true}) {bUsePreviewValueAsDefault=true, description=Inputs a value between 0-1., inputName="Value (0-1)", inputType=FunctionInput_Scalar}
  NamedRerouteDeclaration(None=@n9) {name=Value, nodeColor=<rgba 0 0 0 1>}
 ]
}
