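# Translucent UI material for an Item Shop tile background that shows an animated field of holographic squares.
# A panning grid of rounded boxes fades through an eased vertical gradient broken up by scrolling noise.
# Named reroutes supply tile background colors, flow-map noise, radial gradients, a spotlight and an outline frame.
# Shared Item Shop functions and a parameter collection handle rounded corners.
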Material /ItemShop/UI/Art/HolographicSquares/M_UI_HolographicSquares
graph {
 properties {bAutomaticallySetUsageInEditor=false, blendMode=BLEND_Translucent, materialDomain=MD_UI, usedShadingModels=Default Lit}
 nodes [
  Constant()
  LinearInterpolate(A=NamedRerouteUsage(), B=Divide(A=NamedRerouteUsage()) {constB=1}, Alpha=NamedRerouteUsage())
  n1 = Add(A=NamedRerouteUsage(), B=Multiply(A=NamedRerouteUsage()) {constB=0.25})
  LinearInterpolate(A=NamedRerouteUsage(), B=NamedRerouteUsage(), Alpha=@n1)
  n2 = NamedRerouteUsage()
  n3 = ComponentMask(None=@n2) {g=true}
  LinearInterpolate(Alpha=@n3) {constA=0.1}
  n4 = NamedRerouteUsage()
  MaterialFunctionCall(UVs=@n4, Texture Scale=ScalarParameter() {defaultValue=0.25, group=RadialGradient1, parameterName=R1Scale}) {materialFunction=&ScaleUVsByCenter}
  n5 = MaterialFunctionCall!(UV=Panner(Coordinate=NamedRerouteUsage()) {speedY=0.02}, Tiling X=Constant() {r=22}, Tiling Y=Constant() {r=20}) {materialFunction=/FortniteCoreUI/UI/Community/Art/MaterialFunctions/MF_UI_GridTiling}
  n6 = Constant() {r=0.6}
  n7 = Constant() {r=0.2}
  n8 = NamedRerouteUsage()
  n9 = MaterialFunctionCall(UVs=@n2, Texture Scale=Constant2Vector() {g=5, r=1.5}) {materialFunction=&ScaleUVsByCenter}
  n10 = TextureSample(UVs=Add(A=Multiply(A=@n8) {constB=0.25}, B=Panner(Coordinate=@n9.UVs) {speedX=0.05})) {samplerType=SAMPLERTYPE_Grayscale, texture=&LowResBlurredNoise}
  n11 = Subtract(A=@n3) {constB=0.2}
  n12 = LinearInterpolate(A=Subtract(A=Divide(A=@n10.R) {constB=0.5}) {constB=0.21}, Alpha=OneMinus(None=Divide(A=@n11) {constB=0.25})) {constB=0.08}
  n13 = LinearInterpolate(Alpha=@n3) {constA=-0.6}
  n14 = OneMinus(None=MaterialFunctionCall!("Value (0-1)"=OneMinus(None=@n13)) {materialFunction=/FortUILibrary/MaterialFunctions/Basics/MF_UI_EaseCurves_05_Quint}.Out)
  n15 = LinearInterpolate(
   B=Min(
    A=@n3
    B=MaterialFunctionCall(UVs=@n5.PatternUVs, Size=@n6, CornerRoundness=@n7) {materialFunction=&MF_UI_Box}.Fill
   )
   Alpha=Saturate(None=Add(A=@n14, B=@n12))
  )
  n16 = NamedRerouteUsage()
  Min(
   A=MaterialFunctionCall!(Input Alpha=Constant() {r=1}, bOverrideCornerRadius, Corner Radius Override) {materialFunction=/FortUILibrary/Systems/ItemShopTile/Materials/MaterialFunctions/MF_UI_ItemShop_RoundedCorners}.Final Alpha
   B=Add(A=@n1, B=Add(A=@n15, B=@n16))
  )
  NamedRerouteDeclaration(None=MaterialFunctionCall() {materialFunction=&MF_UI_UVScaling}.ScaleY) {name=ScaletoY, nodeColor=<rgba 0 1 0.0156 1>}
  n17 = Length(None=NamedRerouteUsage()) {is=class MaterialExpressionLength}
  n18 = Multiply(A=NamedRerouteUsage(), B=ScalarParameter() {defaultValue=0.3, group=FlowMap, parameterName=FlowMapTiling})
  n19 = AppendVector(A=ComponentMask(None=@n18) {g=true}, B=ComponentMask(None=@n18) {r=true})
  n20 = ScalarParameter() {defaultValue=0.03, group=FlowMap, parameterName=FlowMapSpeedX}
  n21 = ScalarParameter() {defaultValue=0.05, group=FlowMap, parameterName=FlowMapSpeedY}
  n22 = Frac(None=Multiply(A=Time(), B=AppendVector(A=@n20, B=@n21)) {constB=-0.07})
  n23 = TextureObjectParameter() {
   group=FlowMap
   parameterName=FlowMapTexture
   samplerType=SAMPLERTYPE_Grayscale
   texture=&LowResBlurredNoise
  }
  n24 = Add(A=TextureSample(UVs=Subtract(A=@n19, B=@n22), Tex=@n23).R) {constB=-0.5}
  n25 = Add(A=TextureSample(UVs=Add(A=@n22, B=@n18), Tex=@n23).R) {constB=-0.5}
  NamedRerouteDeclaration(None=Multiply(A=Multiply(A=@n17) {constB=-1}, B=Multiply(A=@n24, B=@n25))) {name=FlowMap, nodeColor=<rgba 0 0.9766 1 1>}
  NamedRerouteDeclaration(None=Divide(A=Constant3Vector() {constant=<rgba 1 0.9352 0.3339 1>}) {constB=0}) {name=Color 2 - Accent, nodeColor=<rgba 0 0.8594 1 1>}
  n26 = ScalarParameter() {defaultValue=5, group=Spotlight, parameterName=SpotlightSizeX}
  n27 = ScalarParameter() {defaultValue=0.6, group=Spotlight, parameterName=SpotlightSizeY}
  n28 = Add(
   A=MaterialFunctionCall(UVs=NamedRerouteUsage(), Texture Scale=AppendVector(A=@n26, B=@n27)) {materialFunction=&ScaleUVsByCenter}.UVs
   B=AppendVector(
    A=ScalarParameter() {defaultValue=-0.2, group=Spotlight, parameterName=SpotlightX}
    B=ScalarParameter() {defaultValue=-0.3, group=Spotlight, parameterName=SpotlightY}
   )
  )
  n29 = Constant()
  n30 = Constant() {r=1}
  NamedRerouteDeclaration(None=MaterialFunctionCall!("Value (0-1)"=MaterialFunctionCall(UV=@n28, GradientStart=@n29, GradientEnd=@n30) {materialFunction=&MF_UI_RadialGradient}.Smooth) {materialFunction=/FortUILibrary/MaterialFunctions/Basics/MF_UI_EaseCurves_06_Circular}.In) {name=Spotlight, nodeColor=<rgba 1 0 0.4687 1>}
  n31 = MaterialFunctionCall!("Value (0-1)"=ComponentMask(None=NamedRerouteUsage()) {g=true}) {materialFunction=/FortUILibrary/MaterialFunctions/Basics/MF_UI_EaseCurves_03_Cubic}
  n32 = CollectionParameter() {collection=/FortUILibrary/Systems/ItemShopTile/MPC_ItemShop, parameterName=CornerRadius}
  n33 = MaterialFunctionCall(
   Corner Radius=MaterialFunctionCall(R=@n32, G=@n32, B=@n32, A=@n32) {materialFunction=&AppendMany}.RGBA
   Radius in Pixels=Constant() {r=1}
   Outline Thickness=Constant() {r=4}
  ) {materialFunction=&MF_UI_CoreUI_RoundedMask}
  n34 = Multiply(A=OneMinus(None=SmoothStep(Value=@n33.Outline SDF) {constMax=0.25})) {constB=0.2}
  NamedRerouteDeclaration(None=LinearInterpolate(A=LinearInterpolate(Alpha=@n31.In-Out) {constA=0.3}, Alpha=Max(A=@n33.Outline Mask, B=@n34))) {name=OutlineFrame, nodeColor=<rgba 0 0.5547 1 1>}
  n35 = MaterialFunctionCall() {materialFunction=/FortUILibrary/Systems/ItemShopTile/Materials/MaterialFunctions/MF_UI_ItemShop_BackgroundParams}
  NamedRerouteDeclaration(None=@n35.Color 1) {name=color 1, nodeColor=<rgba 0.2422 0 1 1>}
  n36 = Add(
   A=@n4
   B=AppendVector(
    A=ScalarParameter() {defaultValue=0.52, group=RadialGradient1, parameterName=R1X}
    B=ScalarParameter() {defaultValue=0.5, group=RadialGradient1, parameterName=R1Y}
   )
  )
  n37 = Constant()
  n38 = Constant() {r=1}
  n39 = Multiply(A=MaterialFunctionCall("Value (0-1)"=MaterialFunctionCall(UV=@n36, GradientStart=@n37, GradientEnd=@n38) {materialFunction=&MF_UI_RadialGradient}.Smooth) {materialFunction=&MF_UI_EaseCurves_06_Circular}.In) {constB=0.25}
  n40 = MaterialFunctionCall(UVs=@n4, Texture Scale=ScalarParameter() {defaultValue=0.75, group=RadialGradient2, parameterName=R2Scale}) {materialFunction=&ScaleUVsByCenter}
  n41 = AppendVector(
   A=ScalarParameter() {defaultValue=-0.62, group=RadialGradient2, parameterName=R2X}
   B=ScalarParameter() {defaultValue=-0.032, group=RadialGradient2, parameterName=R2Y}
  )
  n42 = MaterialFunctionCall("Value (0-1)"=MaterialFunctionCall(UV=Add(A=@n40.UVs, B=@n41), GradientStart=Constant(), GradientEnd=Constant() {r=1}) {materialFunction=&MF_UI_RadialGradient}.Smooth) {materialFunction=&MF_UI_EaseCurves_06_Circular}
  NamedRerouteDeclaration(None=Multiply(A=Max(A=@n39, B=@n42.In), B=ScalarParameter() {defaultValue=0.25, group=RadialGradient1, parameterName=RadialGradientOpacity})) {name=RadialGradients, nodeColor=<rgba 0.1484 1 0 1>}
  NamedRerouteDeclaration(None=MaterialFunctionCall() {materialFunction=&GetUserInterfaceUV}.9-Slice UV) {name=ScaletoFit, nodeColor=<rgba 0.875 1 0 1>}
  NamedRerouteDeclaration(None=Divide(A=@n35.Color 2) {constB=0.5}) {name=color 2, nodeColor=<rgba 0.0781 1 0 1>}
  NamedRerouteDeclaration(None=@n35.Color 3) {name=color 3, nodeColor=<rgba 1 0 0.3984 1>}
  NamedRerouteDeclaration(None=Constant3Vector() {constant=<rgba 1 0.37 0.0104 1>}) {name=color 1 - BAse, nodeColor=<rgba 1 0 0.5391 1>}
  NamedRerouteUsage()
  NamedRerouteUsage()
  NamedRerouteUsage()
  NamedRerouteUsage()
 ]
}
